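; Parent Blueprint for dual-wield claw melee weapons, handling hand attachment, swing trails and slash impact effects.
; On equip it attaches both meshes, binds swing events through a player interface and optionally applies an air-attack GameplayEffect.
; Swing events toggle Niagara trails and record two socket positions per hand.
; Impacts spawn decals and directional FX, oriented by projecting those positions onto a plane facing the player, within a timed grace window.
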(event FortWeapon|EventOnEquippedAndAttached
  (bind _swingfxniagara_offhand (Variables|Default|GetSwingFXNiagara_Offhand))
  ;[1 1 1] On Equip
  (|CallParentFunction)
  (bind _asfort_pawn (Utilities|Casting|CastToFortPawn (Game|GetInstigator)))
  (Utilities|IsValid (Variables|/Script/Engine.Character|GetMesh _asfort_pawn)
    (:Is%20Valid
      _)
    (:Is%20Not%20Valid
      (Utilities|FlowControl|Delay 0.1)))
  (CallFunction|BindDualMeleeSwingEvents)
  (Utilities|FlowControl|Sequence
    (:then_0
      ;[1 1 1] Fix weapon position/rotation on equip
      (Transformation|AttachComponentToComponent (Variables|Visual|GetWeaponMesh) (Variables|/Script/Engine.Character|GetMesh _asfort_pawn) (Variables|WeaponPositionInfo|GetMainhandSocketName) "SnapToTarget" "SnapToTarget" "SnapToTarget" true)
      (Transformation|AttachComponentToComponent (Variables|Default|GetWeaponMesh_Offhand) (Variables|/Script/Engine.Character|GetMesh _asfort_pawn) (Variables|WeaponPositionInfo|GetOffhandSocketName) "SnapToTarget" "SnapToTarget" "SnapToTarget" true)
      (Transformation|SetRelativeLocationAndRotation (Variables|Visual|GetWeaponMesh) (Variables|WeaponPositionInfo|GetWeaponRelativeOffset_Main) (Variables|WeaponPositionInfo|GetWeaponRelativeRotation_Main) false true)
      (Transformation|SetRelativeLocationAndRotation (Variables|Default|GetWeaponMesh_Offhand) (Variables|WeaponPositionInfo|GetWeaponRelativeOffset_Offhand) (Variables|WeaponPositionInfo|GetWeaponRelativeRotation_Offhand) false true)
      ;[1 1 1] Update FX Trails Positions
      (Transformation|AttachComponentToComponent (Variables|Default|GetSwingFXNiagara_Main) (Variables|Visual|GetWeaponMesh) (Variables|Effects|GetSwingFXSocketName) "SnapToTarget" "SnapToTarget" "SnapToTarget" true)
      (Transformation|AttachComponentToComponent (Variables|Default|GetSwingFXNiagara_Offhand) (Variables|Default|GetWeaponMesh_Offhand) (Variables|Effects|GetSwingFXSocketName) "SnapToTarget" "SnapToTarget" "SnapToTarget" true)
      (Utilities|FlowControl|DoOnce false
        (:Completed
          (Transformation|AddLocalOffset (Variables|Default|GetSwingFXNiagara_Offhand) "-65.000000, 0, 0" false false)
          (GameplayCue|ExecuteGameplayCue (Game|GetInstigator) (Variables|Default|GetOnEquipGameplayCue)))))
    (:then_1
      ;[1 1 1] Set player skel mesh and offhand parameters for Niagara
      (Niagara|SetNiagaraSkeletalMeshComponent (Variables|Default|GetSwingFXNiagara_Main) "SkeletalMesh" (Variables|/Script/Engine.Character|GetMesh _asfort_pawn))
      (Niagara|SetNiagaraSkeletalMeshComponent _swingfxniagara_offhand "SkeletalMesh" (Variables|/Script/Engine.Character|GetMesh _asfort_pawn))
      (Effects|Components|ParticleSystem|SetBoolParameter _swingfxniagara_offhand "OffHand" true))
    (:then_2
      (CallFunction|MantisAirAttack))))

(event FortWeapon|EventOnLocalUnEquip
  (bind _returnvalue (Game|GetInstigator))
  ;[1 1 1] On Unequip / Detached
  (|CallParentFunction)
  (Transformation|DetachFromComponent (Variables|Default|GetWeaponMesh_Offhand) "KeepRelative" "KeepRelative" "KeepRelative" true)
  (CallFunction|UnbindDualMeleeSwingEvents)
  (Audio|PlaySoundatLocation (Variables|Sounds|GetUnequipSound) (Transformation|GetActorLocation _returnvalue) :!pinned true) ; Unequip Sound
  (GameplayCue|ExecuteGameplayCue _returnvalue (Variables|Default|GetOnUnequipGameplayCue)))

(event FortWeapon|EventOnUnEquippedAndDetached
  (bind _returnvalue (Game|GetInstigator))
  ;[1 1 1] On Unequip / Detached
  (|CallParentFunction)
  (Transformation|DetachFromComponent (Variables|Default|GetWeaponMesh_Offhand) "KeepRelative" "KeepRelative" "KeepRelative" true)
  (CallFunction|UnbindDualMeleeSwingEvents)
  (Audio|PlaySoundatLocation (Variables|Sounds|GetUnequipSound) (Transformation|GetActorLocation _returnvalue) :!pinned true) ; Unequip Sound
  (GameplayCue|ExecuteGameplayCue _returnvalue (Variables|Default|GetOnUnequipGameplayCue)))

(event Custom|Swing%20Left
  ;[1 1 1] Swing FX
  (Components|Activation|Activate (Variables|Default|GetSwingFXNiagara_Offhand) true))

(event Custom|Swing%20Left%202
  (bind _impactdecalpositionsocketname (Variables|Default|GetImpactDecalPositionSocketName))
  (bind _weaponmesh_offhand (Variables|Default|GetWeaponMesh_Offhand))
  (bind _returnvalue (Transformation|GetSocketLocation _weaponmesh_offhand _impactdecalpositionsocketname))
  (bind _returnvalue_2 (Game|GetInstigator))
  (bind _returnvalue_3 (Transformation|GetActorLocation _returnvalue_2))
  (bind _returnvalue_4 (- _returnvalue _returnvalue_3))
  ;[1 1 1] Swing FX
  (Networking|CanExecuteCosmeticEvents
    (:True
      (if (Variables|Default|GetOffHandFirstDecalLocationIsSet)
        (Utilities|Array|SetArrayElem (Variables|Default|GetOffHandDecalLocations) 1 _returnvalue_4)
        (Variables|Default|SetOffHandFirstDecalLocationIsSet false)
        (Variables|Default|SetOffHandDecalLocationsReady true)
        (CallFunction|TrySpawnDecal)
        (else
          (Utilities|Array|SetArrayElem (Variables|Default|GetOffHandDecalLocations) 0 _returnvalue_4)
          (Variables|Default|SetOffHandFirstDecalLocationIsSet true)
          (Variables|Default|SetOffHandDecalLocationsReady false))))))

(event Custom|MantisAirAttack
  (bind _returnvalue (Game|GetInstigator))
  (bind _returnvalue_2 (Ability|GetAbilitySystemComponent _returnvalue))
  (if (ScalableFloat|GetValueasBool (Variables|Default|GetAirAttackActive))
    (Utilities|IsValid _returnvalue
      (:Is%20Valid
        (Utilities|IsValid _returnvalue_2
          (:Is%20Valid
            (GameplayEffects|ApplyGameplayEffectToTarget :self _returnvalue_2 :GameplayEffectClass (Variables|Default|GetGE_AirAttackActive) :Target _returnvalue_2 :Level 0.0)))))))

(event Custom|Swing%20Left%20End
  ;[1 1 1] Swing FX
  (Components|Activation|Deactivate (Variables|Default|GetSwingFXNiagara_Offhand))
  (Variables|Default|SetOffHandFirstDecalLocationIsSet false)
  (Variables|Default|SetOffHandDecalLocationsReady false))

(event Custom|Swing%20Right
  ;[1 1 1] Swing FX
  (Components|Activation|Activate (Variables|Default|GetSwingFXNiagara_Main) true))

(event Custom|Swing%20Right%202
  (bind _impactdecalpositionsocketname (Variables|Default|GetImpactDecalPositionSocketName))
  (bind _weaponmesh (Variables|Visual|GetWeaponMesh))
  (bind _returnvalue (Transformation|GetSocketLocation _weaponmesh _impactdecalpositionsocketname))
  (bind _returnvalue_2 (Game|GetInstigator))
  (bind _returnvalue_3 (Transformation|GetActorLocation _returnvalue_2))
  (bind _returnvalue_4 (- _returnvalue _returnvalue_3))
  ;[1 1 1] Swing FX
  (Networking|CanExecuteCosmeticEvents
    (:True
      (if (Variables|Default|GetMainHandFirstDecalLocationIsSet)
        (Utilities|Array|SetArrayElem (Variables|Default|GetMainHandDecalLocations) 1 _returnvalue_4)
        (Variables|Default|SetMainHandFirstDecalLocationIsSet false)
        (Variables|Default|SetMainHandDecalLocationsReady true)
        (CallFunction|TrySpawnDecal)
        (else
          (Utilities|Array|SetArrayElem (Variables|Default|GetMainHandDecalLocations) 0 _returnvalue_4)
          (Variables|Default|SetMainHandFirstDecalLocationIsSet true)
          (Variables|Default|SetMainHandDecalLocationsReady false))))))

(event Custom|Swing%20Right%20End
  ;[1 1 1] Swing FX
  (Components|Activation|Deactivate (Variables|Default|GetSwingFXNiagara_Main))
  (Variables|Default|SetMainHandFirstDecalLocationIsSet false)
  (Variables|Default|SetMainHandDecalLocationsReady false))

(event FortWeapon|EventOnPlayImpactFX
  (params (HitResult :type Hit%20Result%20Structure%20%28by%20ref%29) (ImpactPhysicalSurface :type EPhysicalSurface%20Enum :default "SurfaceType_Default") (SpawnedPSC :type FXSystem%20Component%20Object%20Reference))
  (bind (_bblockinghit _binitialoverlap _time _distance _location _impactpoint _normal _impactnormal) (Collision|BreakHitResult HitResult))
  ;[1 1 1] Impact FX/Decal
  (|CallParentFunction HitResult ImpactPhysicalSurface SpawnedPSC)
  (bind _output_get (Variables|WeaponPositionInfo|SetImpactLocation (+ _location (* _impactnormal (Variables|Default|GetDecalOffset)))))
  (if (Variables|Default|GetDebugDecals)
    (Rendering|Debug|DrawDebugSphere _output_get 4.0 6 "(R=0.060320,G=1.000000,B=0.000000,A=1.000000)" 4.0 1.0)
    (else
      _))
  (Variables|Default|SetReadyToAddImpactDecal true)
  (CallFunction|TrySpawnDecal))

(event Custom|SetReadyToAddImpactDecalsFalse
  ;[1 1 1] Set ReadyToAddImpactDecals to False
  (Variables|Default|SetReadyToAddImpactDecal false))

(event Custom|TrySpawnDecal
  (bind _returnvalue (Game|GetInstigator))
  (bind _returnvalue_2 (Transformation|GetActorLocation _returnvalue))
  (bind _returnvalue_3 (Game|GetInstigator))
  (bind _returnvalue_4 (Transformation|GetActorLocation _returnvalue_3))
  (bind _offhanddecallocations (Variables|Default|GetOffHandDecalLocations))
  (bind _decalrotation (CallFunction|CalculateDecalRotationfrom2Locations _offhanddecallocations))
  (bind _mainhanddecallocations (Variables|Default|GetMainHandDecalLocations))
  (bind _decalrotation_2 (CallFunction|CalculateDecalRotationfrom2Locations _mainhanddecallocations))
  (bind _impactlocation (Variables|WeaponPositionInfo|GetImpactLocation))
  (bind _mainhanddecallocations_2 (Variables|Default|GetMainHandDecalLocations))
  (bind _impactlocation_2 (Variables|WeaponPositionInfo|GetImpactLocation))
  (bind _mainhanddecallocations_3 (Variables|Default|GetMainHandDecalLocations))
  ;[1 1 1] Try Spawn Decals And Directional Particles
  ;;[1 1 1] Main Hand Decals
  (if (Variables|Default|GetReadyToAddImpactDecal)
    (if (Variables|Default|GetMainHandDecalLocationsReady)
      (Utilities|FlowControl|Sequence
        (:then_0
          (bind _returnvalue_5 (Rendering|Decal|SpawnDecalatLocation (Variables|Effects|GetImpactDecal) (Variables|Default|GetDecalScale) _impactlocation_2 _decalrotation_2 2.0))
          (Rendering|Components|Decal|SetFadeOut _returnvalue_5 0.0 1.0 false)
          (Variables|Default|SetMainHandDecalLocationsReady false)
          (Utilities|Time|SetTimerbyEvent (EventDispatchers|CreateEvent :!event Custom|SetReadyToAddImpactDecalsFalse) (Variables|Default|GetDualDecalGraceWindow))
          (if (Variables|Default|GetDebugDecals)
            (for _array_element _mainhanddecallocations_3
              (Rendering|Debug|DrawDebugSphere (+ _array_element _returnvalue_4) 4.0 6 "(R=1.000000,G=0.000000,B=0.743275,A=1.000000)" 4.0 1.0))
            (Rendering|Debug|DrawDebugLine (+ _returnvalue_4 (Utilities|Array|Get(aref) _mainhanddecallocations_3)) (+ _returnvalue_4 (Utilities|Array|Get(aref) _mainhanddecallocations_3 1)) "(R=1.000000,G=0.000000,B=0.883718,A=1.000000)" 4.0 1.0)
            (else
              _)))
        (:then_1
          (Effects|Components|ParticleSystem|SpawnFXAtLocation (Variables|Effects|GetMainHandDirectionalImpactFX) _impactlocation_2 _decalrotation_2)))
      (else
        _))
    ;;[1 1 1] Off Hand Decals
    (if (Variables|Default|GetOffHandDecalLocationsReady)
      (Utilities|FlowControl|Sequence
        (:then_0
          (bind _returnvalue_6 (Rendering|Decal|SpawnDecalatLocation (Variables|Effects|GetImpactDecal) (Variables|Default|GetDecalScale) _impactlocation _decalrotation 2.0))
          (Rendering|Components|Decal|SetFadeOut _returnvalue_6 0.0 1.0 false)
          (Variables|Default|SetOffHandDecalLocationsReady false)
          (if (Variables|Default|GetDebugDecals)
            (for _array_element_2 _mainhanddecallocations_2
              (Rendering|Debug|DrawDebugSphere (+ _array_element_2 _returnvalue_2) 4.0 6 "(R=1.000000,G=0.805396,B=0.000000,A=1.000000)" 4.0 1.0))
            (Rendering|Debug|DrawDebugLine (+ _returnvalue_2 (Utilities|Array|Get(aref) _mainhanddecallocations_2)) (+ _returnvalue_2 (Utilities|Array|Get(aref) _mainhanddecallocations_2 1)) "(R=1.000000,G=0.933952,B=0.000000,A=1.000000)" 4.0 1.0)
            (else
              _))
          (Utilities|Time|SetTimerbyEvent (EventDispatchers|CreateEvent :!event Custom|SetReadyToAddImpactDecalsFalse) (Variables|Default|GetDualDecalGraceWindow)))
        (:then_1
          (Effects|Components|ParticleSystem|SpawnFXAtLocation (Variables|Effects|GetOffhandDirectionalImpactFX) _impactlocation _decalrotation))))))

(fn BindDualMeleeSwingEvents
  (bind _asmelee_swing_and_anim_trail_events_player_interface (Utilities|Casting|CastToMeleeSwingAndAnimTrailEventsPlayerInterface (Game|GetInstigator))
    (:then
      (MeleeSwingAndAnimTrailEventsPlayerInterface|BindtoSwingFX _asmelee_swing_and_anim_trail_events_player_interface (EventDispatchers|CreateEvent :!event Swing%20Left) (EventDispatchers|CreateEvent :!event Swing%20Left%20End) (EventDispatchers|CreateEvent :!event Swing%20Right) (EventDispatchers|CreateEvent :!event Swing%20Right%20End) :!class /Script/FortniteGame.MeleeSwingAndAnimTrailEventsPlayerInterface)
      (MeleeSwingAndAnimTrailEventsPlayerInterface|BindtoSwingFX2 _asmelee_swing_and_anim_trail_events_player_interface (EventDispatchers|CreateEvent :!event Swing%20Left%202) (EventDispatchers|CreateEvent :!event Swing%20Right%202) :!class /Script/FortniteGame.MeleeSwingAndAnimTrailEventsPlayerInterface))
    (:CastFailed
      (Development|PrintString (Utilities|String|Append "Cannot bind events.  Cast Failed!  Instigator of dual-wield melee weapon is not a MeleeSwingAndAnimTrailEventsPlayerInterface, so it will not dual-wield properly. Instigator: " (Utilities|String|ToString(Object) (Game|GetInstigator)))) ; Instigator class is not supported!
    )))

(fn CalculateDecalRotationFrom2Locations
  (params (DecalLocations :type Array%20of%20Vectors))
  (bind _returnvalue (Game|GetInstigator))
  (bind _returnvalue_2 (Transformation|GetActorForwardVector _returnvalue))
  (bind _impactlocation (Variables|WeaponPositionInfo|GetImpactLocation))
  (bind _returnvalue_3 (Transformation|GetActorLocation _returnvalue))
  (bind _output (Utilities|Array|Get(aref) DecalLocations 1))
  (bind _returnvalue_4 (+ _output _returnvalue_3))
  (bind _returnvalue_5 (Math|Vector|ProjectPointontoPlane _returnvalue_4 _impactlocation _returnvalue_2))
  (bind _impactlocation_2 (Variables|WeaponPositionInfo|GetImpactLocation))
  (bind _output_2 (Utilities|Array|Get(aref) DecalLocations))
  (bind _returnvalue_6 (+ _output_2 _returnvalue_3))
  (bind _returnvalue_7 (Math|Vector|ProjectPointontoPlane _returnvalue_6 _impactlocation_2 _returnvalue_2))
  ;[1 1 1] Project impact points onto a plane perpendicular to the player's facing. Using the angle between those points and the player's facing, determine rotation of decal
  (if (Variables|Default|GetDebugDecals)
    (Rendering|Debug|DrawDebugLine _returnvalue_7 _returnvalue_5 "(R=0.279911,G=1.000000,B=0.000000,A=1.000000)" 4.0 1.0)
    (else
      _))
  (return (Math|Rotator|MakeRotfromXY _returnvalue_2 (Math|Vector|Normalize (- _returnvalue_7 _returnvalue_5)))))

(fn UnbindDualMeleeSwingEvents)

(fn ConstructionScript
  (|CallParentFunction))
